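# Sound cue for the Core AR's aim-down-sights-in foley, played as a weapon foley sound with reload-style attenuation.
# A mixer layers a randomly picked military rifle ADS sound with a quieter, slightly delayed random hand-grab sound.
# Modulators sit on both layers for playback variation, with a wider volume range set on the hand-grab layer.
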SoundCue /FlipperGameplay/Items/Weapons/CoreAR/Audio/Core_AR_ADS_IN_Cue
root=Mixer_0
nodes [
 {id=Delay_0, children [Modulator_2], properties {delayMax=0.1, delayMin=0.1}}
 {id=Mixer_0, children [Modulator_0,Delay_0], properties {inputVolume [1,0.6]}}
 {id=Modulator_0, children [Random_1]}
 {id=Modulator_2, children [Random_0], properties {volumeMax=1, volumeMin=0.6}}
 {id=Random_0, children [WavePlayer_1,WavePlayer_2,WavePlayer_3], properties {bShouldExcludeFromBranchCulling=true, weights [1,1,1]}}
 {id=Random_1, children [WavePlayer_4,WavePlayer_5], properties {weights [1,1]}}
 {id=WavePlayer_1, properties {soundWaveAssetPtr=&/Gun_Grab_Hand_ADS_In_01}}
 {id=WavePlayer_2, properties {soundWaveAssetPtr=&/Gun_Grab_Hand_ADS_In_02}}
 {id=WavePlayer_3, properties {soundWaveAssetPtr=&/Gun_Grab_Hand_ADS_In_03}}
 {id=WavePlayer_4, properties {soundWaveAssetPtr=/Game/Sounds/Fort_Weapons/ADS/Weapon_MILRIFLE_ADS_IN_01}}
 {id=WavePlayer_5, properties {soundWaveAssetPtr=/Game/Sounds/Fort_Weapons/ADS/Weapon_MILRIFLE_ADS_IN_02}}
]
properties {
 attenuationSettings=&Fort_Weapon_Reload_Att
 maxDistance=2250
 soundClassObject=&Fort_WeaponFoley
 volumeMultiplier=0.376
}
