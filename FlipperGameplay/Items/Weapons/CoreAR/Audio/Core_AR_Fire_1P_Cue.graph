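# Sound cue for the first-person, close-range gunfire of the Core AR weapon.
# A Random node picks among six equally weighted shot waves and is excluded from branch culling.
# Playback routes through the ranged-weapon 1P fire sound class, with shared first-person attenuation and gunfire concurrency assets at slightly reduced volume.
# The cue's debug flag is left on.
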SoundCue /FlipperGameplay/Items/Weapons/CoreAR/Audio/Core_AR_Fire_1P_Cue
root=Random_3
nodes [
 {
  id=Random_3
  children [WavePlayer_10, WavePlayer_11, WavePlayer_12, WavePlayer_13, WavePlayer_14, WavePlayer_15]
  properties {bShouldExcludeFromBranchCulling=true, weights [1,1,1,1,1,1]}
 }
 {id=WavePlayer_10, properties {soundWaveAssetPtr=&/Core_AR_Fire_Close_1P_01}}
 {id=WavePlayer_11, properties {soundWaveAssetPtr=&/Core_AR_Fire_Close_1P_02}}
 {id=WavePlayer_12, properties {soundWaveAssetPtr=&/Core_AR_Fire_Close_1P_03}}
 {id=WavePlayer_13, properties {soundWaveAssetPtr=&/Core_AR_Fire_Close_1P_04}}
 {id=WavePlayer_14, properties {soundWaveAssetPtr=&/Core_AR_Fire_Close_1P_05}}
 {id=WavePlayer_15, properties {soundWaveAssetPtr=&/Core_AR_Fire_Close_1P_06}}
]
properties {
 attenuationSettings=&Fort_Weapon_Projectile_FirstPerson_Athena_Att
 bDebug=true
 maxDistance=2097152
 soundClassObject=&Fort_WeaponRanged_Fire_1P_PSM
 volumeMultiplier=0.85
 attenuationOverrides {manualReverbSendLevel=-1.0165, reverbSendMethod=Manual}
 concurrencyOverrides {bLimitToOwner=true, maxCount=2, resolutionRule=StopOldest, volumeScale=0.9}
 concurrencySet [&Weapon_Gunfire_FirstPerson_Conc]
}
