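# Sound cue for Core AR weapon fire, layering close, mid and distant recordings.
# A mixer blends three branches, each with its own sound class override and distance attenuation, over randomly picked wave variations.
# Only the distant layer adds slight pitch variation, and the close layer uses a separate third-person fire class.
# A gunfire concurrency set is assigned alongside per-owner stop-oldest override settings, and debug is enabled.
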SoundCue /FlipperGameplay/Items/Weapons/CoreAR/Audio/Core_AR_Fire_Cue
root=Mixer_17
nodes [
 {
  id=Attenuation_0
  children [Modulator_1]
  properties {attenuationSettings=&Fort_Weapon_Projectile_Medium_Mono_Athena_Att}
 }
 {
  id=Attenuation_34
  children [Modulator_4]
  properties {attenuationSettings=&Fort_Weapon_Projectile_Distant_Mono_Athena_Att}
 }
 {id=Attenuation_49, children [Modulator_5], properties {attenuationSettings=&Fort_Weapon_Projectile_Close_Athena_Att}}
 {id=Mixer_17, children [SoundClass_32, SoundClass_0, SoundClass_35], properties {inputVolume [0.95,0.7,0.65]}}
 {id=Modulator_1, children [Random_7], properties {volumeMax=1, volumeMin=1}}
 {id=Modulator_4, children [Random_8], properties {pitchMax=1.01, pitchMin=0.99, volumeMax=1, volumeMin=1}}
 {id=Modulator_5, children [Random_6], properties {volumeMax=1, volumeMin=1}}
 {
  id=Random_6
  children [WavePlayer_0,WavePlayer_1,WavePlayer_2,WavePlayer_3,WavePlayer_4,WavePlayer_5]
  properties {bShouldExcludeFromBranchCulling=true, weights [1,1,1,1,1,1]}
 }
 {
  id=Random_7
  children [WavePlayer_6, WavePlayer_7, WavePlayer_8, WavePlayer_9, WavePlayer_22, WavePlayer_23]
  properties {bShouldExcludeFromBranchCulling=true, weights [1,1,1,1,1,1]}
 }
 {
  id=Random_8
  children [WavePlayer_24, WavePlayer_25, WavePlayer_26, WavePlayer_27]
  properties {bShouldExcludeFromBranchCulling=true, weights [1,1,1,1]}
 }
 {
  id=SoundClass_0
  children [Attenuation_0]
  properties {soundClassOverride=&Fort_WeaponRanged_Fire_Distant}
 }
 {
  id=SoundClass_32
  children [Attenuation_49]
  properties {soundClassOverride=&Fort_WeaponRanged_Fire_3P_PSM}
 }
 {
  id=SoundClass_35
  children [Attenuation_34]
  properties {soundClassOverride=&Fort_WeaponRanged_Fire_Distant}
 }
 {id=WavePlayer_0, properties {soundWaveAssetPtr=&/Core_AR_Fire_Close_01}}
 {id=WavePlayer_1, properties {soundWaveAssetPtr=&/Core_AR_Fire_Close_02}}
 {id=WavePlayer_2, properties {soundWaveAssetPtr=&/Core_AR_Fire_Close_03}}
 {id=WavePlayer_22, properties {soundWaveAssetPtr=&/Core_AR_Fire_Mid_05}}
 {id=WavePlayer_23, properties {soundWaveAssetPtr=&/Core_AR_Fire_Mid_06}}
 {id=WavePlayer_24, properties {soundWaveAssetPtr=&/Core_AR_Fire_Distant_01}}
 {id=WavePlayer_25, properties {soundWaveAssetPtr=&/Core_AR_Fire_Distant_02}}
 {id=WavePlayer_26, properties {soundWaveAssetPtr=&/Core_AR_Fire_Distant_03}}
 {id=WavePlayer_27, properties {soundWaveAssetPtr=&/Core_AR_Fire_Distant_04}}
 {id=WavePlayer_3, properties {soundWaveAssetPtr=&/Core_AR_Fire_Close_04}}
 {id=WavePlayer_4, properties {soundWaveAssetPtr=&/Core_AR_Fire_Close_05}}
 {id=WavePlayer_5, properties {soundWaveAssetPtr=&/Core_AR_Fire_Close_06}}
 {id=WavePlayer_6, properties {soundWaveAssetPtr=&/Core_AR_Fire_Mid_01}}
 {id=WavePlayer_7, properties {soundWaveAssetPtr=&/Core_AR_Fire_Mid_02}}
 {id=WavePlayer_8, properties {soundWaveAssetPtr=&/Core_AR_Fire_Mid_03}}
 {id=WavePlayer_9, properties {soundWaveAssetPtr=&/Core_AR_Fire_Mid_04}}
]
properties {
 bDebug=true
 maxDistance=27500
 volumeMultiplier=0.85
 concurrencyOverrides {bLimitToOwner=true, maxCount=2, resolutionRule=StopOldest}
 concurrencySet [&Weapon_Gunfire_FirstPerson_Conc]
}
