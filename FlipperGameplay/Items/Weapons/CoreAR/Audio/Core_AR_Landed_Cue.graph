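# Sound cue for the Core AR weapon landing, playing a heavy rifle drop impact sound.
# It picks one of three equally weighted generic impact waves at random, then applies slight pitch modulation for variety.
# It overrides attenuation for short-range falloff and limits concurrency to a few instances, blocking new plays beyond that, under the loot collect sound class.
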SoundCue /FlipperGameplay/Items/Weapons/CoreAR/Audio/Core_AR_Landed_Cue
root=Modulator_0
nodes [
 {id=Modulator_0, children [Random_0], properties {pitchMax=1.1, pitchMin=0.9}}
 {id=Random_0, children [WavePlayer_1,WavePlayer_2,WavePlayer_3], properties {weights [1,1,1]}}
 {id=WavePlayer_1, properties {soundWaveAssetPtr=/Game/Sounds/Fort_UI_Sounds/Gameplay/Weapon_Generic_Drop_Impact_Rifle_Hv_01}}
 {id=WavePlayer_2, properties {soundWaveAssetPtr=/Game/Sounds/Fort_UI_Sounds/Gameplay/Weapon_Generic_Drop_Impact_Rifle_Hv_02}}
 {id=WavePlayer_3, properties {soundWaveAssetPtr=/Game/Sounds/Fort_UI_Sounds/Gameplay/Weapon_Generic_Drop_Impact_Rifle_Hv_03}}
]
properties {
 attenuationSettings=&Foley_Attenuation_Basic_Att
 bOverrideAttenuation=true
 bOverrideConcurrency=true
 maxDistance=1000
 soundClassObject=&Fort_Loot_Collect
 volumeMultiplier=1
 attenuationOverrides {falloffDistance=1000, attenuationShapeExtents {x=0}}
 concurrencyOverrides {maxCount=3, resolutionRule=PreventNew}
}
