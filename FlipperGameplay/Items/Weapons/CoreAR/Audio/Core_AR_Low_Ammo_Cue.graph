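# Sound cue for the low-ammo fire sound of the Core AR weapon in FlipperGameplay.
# A single wave player outputs the Core_AR_Fire_Low_Ammo_01 wave at reduced volume, with no randomization or mixing.
# It uses close weapon projectile attenuation and a low-ammo ranged weapon sound class.
# Its concurrency override prevents new instances from starting once the limit is reached.
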SoundCue /FlipperGameplay/Items/Weapons/CoreAR/Audio/Core_AR_Low_Ammo_Cue
root=WavePlayer_0
nodes [{id=WavePlayer_0, properties {soundWaveAssetPtr=&/Core_AR_Fire_Low_Ammo_01}}]
properties {
 attenuationSettings=&Fort_Weapon_Projectile_Close_Athena_Att
 bOverrideConcurrency=true
 maxDistance=4950
 soundClassObject=/Game/Sounds/FortSoundClassesAndModes/Fort_WeaponRanged_LowAmmo_Degraded_PSM
 volumeMultiplier=0.6
 concurrencyOverrides {resolutionRule=PreventNew}
}
