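# Sound cue for the pickup sound of the Core AR weapon.
# A single wave player plays Core_AR_Pickup_01 through a Modulator left at default settings, with the cue volume reduced to about half.
# It uses the shared weapon pickup attenuation, the loot collect sound class and the PickupMaximum_Oldest concurrency set to limit overlapping pickup sounds.
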SoundCue /FlipperGameplay/Items/Weapons/CoreAR/Audio/Core_AR_Pickup_Cue
root=Modulator_0
nodes [
 {id=Modulator_0, children [WavePlayer_0]}
 {id=WavePlayer_0, properties {soundWaveAssetPtr=&/Core_AR_Pickup_01}}
]
properties {
 attenuationSettings=&Fort_Weapon_Pickup_Att
 maxDistance=3500
 soundClassObject=&Fort_Loot_Collect
 volumeMultiplier=0.5077
 concurrencyOverrides {maxCount=1, resolutionRule=PreventNew}
 concurrencySet [&PickupMaximum_Oldest]
}
