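# Sound cue for the start of the Core AR weapon's reload.
# A single wave player feeds the Core_AR_Reload_Start_01 sound through a Modulator node at the root.
# The cue plays at reduced volume in the weapon foley sound class, using the shared weapon reload attenuation.
# It overrides concurrency and has its debug flag switched on.
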SoundCue /FlipperGameplay/Items/Weapons/CoreAR/Audio/Core_AR_Reload_Start_Cue
root=Modulator_0
nodes [
 {id=Modulator_0, children [WavePlayer_29]}
 {id=WavePlayer_29, properties {soundWaveAssetPtr=&/Core_AR_Reload_Start_01}}
]
properties {
 attenuationSettings=&Fort_Weapon_Reload_Att
 bDebug=true
 bOverrideConcurrency=true
 maxDistance=2250
 soundClassObject=&Fort_WeaponFoley
 volumeMultiplier=0.501
}
