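# Sound cue named for the Core AR scope zoom-in foley, but it has no nodes and a null root, so it produces no audio as it stands.
# It keeps only playback settings: the weapon foley sound class, a low volume, overridden attenuation alongside a referenced basic foley attenuation asset, and a concurrency override that prevents new instances at the limit.
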SoundCue /FlipperGameplay/Items/Weapons/CoreAR/Audio/Core_AR_Scope_ZoomIn_Cue
root=null
nodes []
properties {
 attenuationSettings=&Foley_Attenuation_Basic_Att
 bOverrideAttenuation=true
 bOverrideConcurrency=true
 maxDistance=2700
 soundClassObject=&Fort_WeaponFoley
 volumeMultiplier=0.15
 attenuationOverrides {falloffDistance=1500, nonSpatializedRadiusStart=200, attenuationShapeExtents {x=1200}}
 concurrencyOverrides {resolutionRule=PreventNew}
}
